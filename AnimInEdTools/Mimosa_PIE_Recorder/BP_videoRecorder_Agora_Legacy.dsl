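; Legacy PIE video recorder Blueprint for previewing Fortnite Athena skins through a face camera and scene-capture game camera.
; Construction positions the camera parent, snaps the facecam to the dummy's head and chooses the game cam's render target from the additional 3/4 view option.
; A buttonPresses event polls number keys to force a skin from an array and show its name in a widget.
; Numpad plus and minus lead to camera distance changes.
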(event Custom|buttonPresses
  ;[1 1 1] overwrite current skin with selected from array
  (Variables|Default|GetplayerController
    (:then
      (Utilities|FlowControl|Sequence
        (:then_0
          (CallFunction|SetSkin 0 "One"))
        (:then_1
          (CallFunction|SetSkin 1 "Two"))
        (:then_2
          (CallFunction|SetSkin 2 "Three"))
        (:then_3
          (CallFunction|SetSkin 3 "Four"))
        (:then_4
          (CallFunction|SetSkin 4 "Five"))
        (:then_5
          (CallFunction|SetSkin 5 "Six"))
        (:then_6
          (CallFunction|SetSkin 6 "Seven"))
        (:then_7
          (CallFunction|SetSkin 7 "Eight"))
        (:then_8
          (CallFunction|SetSkin 8 "Nine"))
        (:then_9
          (CallFunction|SetSkin 9 "Zero"))
        (:then_11
          ;[1 1 1] Comment
          (if (Game|Player|WasInputKeyJustPressed (Variables|Default|GetplayerController) "Multiply")
            _)
          (Utilities|FlowControl|DoOnce false))
        (:then_12
          (if (Game|Player|WasInputKeyJustPressed (Variables|Default|GetplayerController) "Add")
            _)
          (Utilities|FlowControl|DoOnce false
            (:Completed
              (CallFunction|IncreaseCamDistance))))
        (:then_13
          (if (Game|Player|WasInputKeyJustPressed (Variables|Default|GetplayerController) "Subtract")
            _)
          (Utilities|FlowControl|DoOnce false
            (:Completed
              (CallFunction|DecreaseCamDistance))))))
    (:else)))

(fn setCamDistance
  (params (camera :type Scene%20Component%20Object%20Reference) (offset :type Float%20%28double-precision%29))
  (Transformation|AddLocalOffset camera (Math|Vector|MakeVector offset) false false))

(fn setSkin
  (params (skinNumber :type Integer) (buttonPressed :type Key%20Structure))
  (bind _skins (Variables|Properties|Getskins))
  (bind _output (Utilities|Array|Get(acopy) _skins skinNumber))
  (if (Game|Player|WasInputKeyJustPressed (Variables|Default|GetplayerController) buttonPressed)
    (Utilities|IsValid _output
      (:Is%20Valid
        (Items|ForceAthenaCosmetic _output)
        (bind _output_get (Variables|Default|SetcurrentSkin _output))
        (CallFunction|SetSkinName _output_get)))))

(fn setSkinName
  (params (skin :type Fort%20Item%20Definition%20Object%20Reference))
  (Widget|SetText(TextBox) (Variables|Getskinname (Variables|Default|GetwidgetPlayer)) (ItemDefinition|GetDisplayName skin)))

(fn ConstructionScript
  (bind _facecamparent (Variables|Default|GetfaceCamParent))
  ;[1 1 1] on construction, remove the bottom right 3/4R camera if desired.
  (Variables|/Script/Engine.SceneCaptureComponent2D|SetTextureTarget (select (Variables|Properties|Getadditional3/4) "" "") (Variables|Default|GetgameCam))
  ;[1 1 1] set location of camera parent for preview
  (Transformation|SetRelativeLocation (Variables|Default|GetcameraParent) "0.000000,0.000000,60.000000" false false)
  ;[1 1 1] attach facecam to dummy for preview setup.
  (Transformation|AttachComponentToComponent _facecamparent (Variables|Default|GetcharBody) "head" "SnapToTarget" "SnapToTarget" "SnapToTarget" true)
  (Transformation|SetRelativeRotation _facecamparent "180.000000,90.000000,-90.000000" false false)
  (Transformation|SetRelativeLocation _facecamparent "4.000000,120.000000,0.000000" false false))
