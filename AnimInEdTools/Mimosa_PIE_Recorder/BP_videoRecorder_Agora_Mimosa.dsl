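; Editor PIE recording rig that films the player and their Mimosa companion through scene-capture cameras shown in a viewport widget.
; A looping timeline finds the companion, attaches body, face and pet-distance components to it, follows the player, and forces time of day.
; The widget displays companion speed, montage, frame, section and skin name, plus a petting-range indicator.
; Polled hotkeys send mood commands, toggle companions via console, switch views and adjust camera distance.
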(event EventBeginPlay
  (bind _companion (Variables|Default|Getcompanion))
  (bind _returnvalue (Actor|GetComponentbyClass _companion "/Script/Engine.SkeletalMeshComponent"))
  (bind _returnvalue_2 (Components|SkeletalMesh|GetAnimInstance _returnvalue))
  (bind _returnvalue_3 (Animation|Montage|GetCurrentActiveMontage _returnvalue_2))
  (bind (_cosmeticloadoutpc_bannericonid _cosmeticloadoutpc_bannercolorid _cosmeticloadoutpc_skydivecontrail _cosmeticloadoutpc_glider _cosmeticloadoutpc_pickaxe _cosmeticloadoutpc_character _cosmeticloadoutpc_charactervariantchannels _cosmeticloadoutpc_backpack _cosmeticloadoutpc_shoes _cosmeticloadoutpc_mimosa) (Utilities|Struct|BreakFortAthenaLoadout (Variables|/Script/FortniteGame.FortPlayerController|GetCosmeticLoadoutPC_BannerIconId (Variables|Default|GetplayerController))))
  (bind _cameraparent (Variables|Default|GetcameraParent))
  (bind _facecamparent (Variables|Default|GetfaceCamParent))
  (bind _widgetplayer (Variables|Default|GetwidgetPlayer))
  ;[1 1 1] on start, remove the preview setup.
  (Components|DestroyComponent (Variables|Default|GetcharBody))
  (Components|DestroyComponent (Variables|Default|GetM_MED_SitPlant_Quadruped))
  (Variables|Default|SetcompanionFound false)
  (Transformation|SetRelativeLocation (Variables|Default|GetcameraParent) "0, 0, 0" false false)
  (Transformation|SetRelativeLocation (Variables|Default|GetOffsetGameCam) (-5e+02 0.0 72.0) false false)
  (Timeline|tick Play
    (:Update
      ;[0.4844 0.4844 0.4844] Check for Companions. If it finds one, setup variables and attach cameras
      (if (not (Variables|Default|GetcompanionFound))
        (bind _returnvalue_4 (Actor|GetActorOfClass 0))
        (Utilities|IsValid _returnvalue_4
          (:Is%20Valid
            (Development|PrintString "AgoraTool: Companion Found!!!!!" true true "(R=0.000000,G=0.660000,B=1.000000,A=1.000000)" 5.0)
            (bind _output_get (Variables|Default|Setcompanion _returnvalue_4))
            (Utilities|IsValid _output_get
              (:Is%20Valid
                (Variables|Default|SetcompanionName (Utilities|String|ToString(Text) (ItemDefinition|GetDisplayName _cosmeticloadoutpc_mimosa)))
                ;[0.4844 0.4844 0.4844] PettingDistance_Setup
                (Transformation|AttachComponentToComponent (Variables|Default|GetCompanionPetDistanceBubble) (Actor|GetComponentbyClass _output_get "/Script/Engine.SkeletalMeshComponent") "None" "KeepRelative" "SnapToTarget" "KeepRelative" true)
                (Utilities|IsValid _widgetplayer
                  (:Is%20Valid
                    (Widget|SetVisibility (Variables|/AnimInEdTools/Mimosa_PIE_Recorder/WIDG_recorder_PlayerView_Mimosa.WIDG_recorder_PlayerView_Mimosa_C|GetPetDistanceIndicator _widgetplayer) "Hidden"))
                  (:Is%20Not%20Valid
                    _))
                ;;[0.8524 0.8524 0.8524] Attach Cameras
                (Transformation|AttachComponentToComponent _cameraparent (Actor|GetComponentbyClass _output_get "/Script/Engine.SkeletalMeshComponent") "None" "KeepRelative" "SnapToTarget" "KeepRelative" true)
                (Transformation|SetRelativeLocation _cameraparent "0.000000,0.000000,15.000000" false false)
                (Transformation|DetachFromComponent _facecamparent "KeepRelative" "KeepRelative" "KeepRelative" true)
                (Transformation|AttachComponentToComponent _facecamparent (Actor|GetComponentbyClass _output_get "/Script/Engine.SkeletalMeshComponent") "face_attach" "KeepRelative" "KeepRelative" "KeepRelative" true)
                (Transformation|SetRelativeRotation _facecamparent "180.000000,90.000000,180.000000" false false)
                (Transformation|SetRelativeLocation _facecamparent "0.000000,150.000000,0.000000" false false)
                (Variables|Default|SetcompanionFound true))))
          (:Is%20Not%20Valid
            _)))
      ;[0.4844 0.4844 0.4844] on tick, set player cam to player location, update widget information
      ;;[1 1 1] watch for button presses.
      (CallFunction|ButtonPresses)
      (Utilities|FlowControl|Sequence
        (:then_0
          ;;[1 1 1] if PC or Actor are invalid, set these variables first. Safety net.
          (bind _playercontroller (Variables|Default|GetplayerController)
            (:then
              (bind _actor (Variables|Default|Getactor)
                (:then
                  ;;[1 1 1] if is NOT facecam, snap the camera rig to the characters root.
                  (if (not (Variables|Properties|GetisFaceCam))
                    (Transformation|SetWorldTransform (Variables|Default|GetScene) (Math|Transform|MakeTransform (Transformation|GetActorLocation _actor) (Pawn|GetControlRotation _playercontroller :!class /Script/Engine.Controller)) false false))
                  ;;[1 1 1] Fill in all relevant text information on the UI
                  (bind _widgetplayer_2 (Variables|Default|GetwidgetPlayer)
                    (:then
                      (Utilities|IsValid _companion
                        (:Is%20Valid
                          (Widget|SetText(TextBox) (Variables|/AnimInEdTools/Mimosa_PIE_Recorder/WIDG_recorder_PlayerView_Mimosa.WIDG_recorder_PlayerView_Mimosa_C|Getspeed _widgetplayer_2) (Utilities|Text|ToText(Float) (Math|Vector|VectorLength (Transformation|GetVelocity _companion))))
                          (Utilities|IsValid _returnvalue_2
                            (:Is%20Valid
                              (Widget|SetText(TextBox) (Variables|/AnimInEdTools/Mimosa_PIE_Recorder/WIDG_recorder_PlayerView_Mimosa.WIDG_recorder_PlayerView_Mimosa_C|GetanimName _widgetplayer_2) (Utilities|Text|ToText(String) (Utilities|GetDisplayName _returnvalue_3)))
                              (Widget|SetText(TextBox) (Variables|/AnimInEdTools/Mimosa_PIE_Recorder/WIDG_recorder_PlayerView_Mimosa.WIDG_recorder_PlayerView_Mimosa_C|Getframe _widgetplayer_2) (Utilities|Text|ToText(Float) (* 3e+01 (Animation|Montage|MontageGetPosition _returnvalue_2 _returnvalue_3))))
                              (Widget|SetText(TextBox) (Variables|/AnimInEdTools/Mimosa_PIE_Recorder/WIDG_recorder_PlayerView_Mimosa.WIDG_recorder_PlayerView_Mimosa_C|GetsectionName _widgetplayer_2) (Utilities|Text|ToText(Name) (Animation|Montage|MontageGetCurrentSection _returnvalue_2 _returnvalue_3)))
                              (Widget|SetText(TextBox) (Variables|/AnimInEdTools/Mimosa_PIE_Recorder/WIDG_recorder_PlayerView_Mimosa.WIDG_recorder_PlayerView_Mimosa_C|Getskinname _widgetplayer_2) (Utilities|Text|ToText(String) (Utilities|String|Split (Utilities|GetDisplayName _returnvalue_2) "_"))))))))
                    (:else)))
                (:else
                  (CallFunction|SetCamera))))
            (:else
              (CallFunction|SetVariables))))
        (:then_1
          (Development|ExecuteConsoleCommand (Utilities|String|Append "Cheat SetTimeOfDay " (Utilities|String|ToString(Integer) (Variables|Properties|GettimeOfDay)))))))))

(event Custom|toggleFacecam
  (bind _scene (Variables|Default|GetScene))
  (bind _offsetgamecam (Variables|Default|GetOffsetGameCam))
  (bind _scene_2 (Variables|Default|GetScene))
  ;[0.5885 0.5885 0.5885] adjusts the location and parent of the top right (Game) camera to focus on the face or not, as desired.
  ;;[1 1 1] on toggle, communicate with the recorder widget to adjust, and sets location and parent of game cam.
  (Class|WIDGRecorderPlayerViewMimosa|ToggleFaceCam (Variables|Default|GetwidgetPlayer))
  (bind _isa (Utilities|FlowControl|FlipFlop)
    (:A
      (Transformation|AttachComponentToComponent _scene_2 (Actor|GetComponentbyClass (Variables|Default|Getactor) "/Script/Engine.SkeletalMeshComponent") "face_attach" "SnapToTarget" "SnapToTarget" "SnapToTarget" true)
      (Transformation|SetRelativeRotation _scene_2 "0,90.000000,-90.000000" false false))
    (:B
      (Transformation|AttachComponentToComponent _scene (Variables|Default|GetDefaultSceneRoot) "None" "SnapToTarget" "SnapToTarget" "SnapToTarget" true)
      (Transformation|SetRelativeRotation _scene "0, 0, 0" false false)))
  ;;[1 1 1] based on is or is not facecam, adjusts the 3/4R and 3/4L position and FOV to match.
  (Variables|Properties|SetisFaceCam _isa)
  (Transformation|SetRelativeLocation _offsetgamecam (select _isa "60.000000, 0,4.000000" "-500.000000,0.000000,72.000000") false false)
  (Transformation|SetRelativeRotation _offsetgamecam (select _isa "0,180.000000, 0" "0, 0, 0") false false)
  (Variables|/Script/Engine.SceneCaptureComponent2D|SetFOVAngle (select _isa 5e+01 8e+01) (Variables|Default|Getscene3/4R))
  (Transformation|SetRelativeLocation (Variables|Default|GetOffset3/4R) (select _isa "150.000000,120.000000,84.044678" "234.535339,236.812073,93.231949") false false)
  (Variables|/Script/Engine.SceneCaptureComponent2D|SetFOVAngle (select _isa 5e+01 8e+01) (Variables|Default|Getscene3/4L))
  (Transformation|SetRelativeLocation (Variables|Default|GetOffset3/4L) (select _isa "150.000000,-120.000000,84.044678" "234.535339,-236.812057,93.231949") false false))

(event OnComponentEndOverlap(CompanionPetDistanceBubble)
  (params (OverlappedComponent :type Primitive%20Component%20Object%20Reference) (OtherActor :type Actor%20Object%20Reference) (OtherComp :type Primitive%20Component%20Object%20Reference) (OtherBodyIndex :type Integer))
  (bind _widgetplayer (Variables|Default|GetwidgetPlayer))
  ;[0.4844 0.4844 0.4844] PettingDistance_Setup
  (Utilities|IsValid _widgetplayer
    (:Is%20Valid
      (Utilities|Casting|CastToPlayerController OtherActor
        (:then
          _)
        (:CastFailed
          _))
      (Widget|SetVisibility (Variables|/AnimInEdTools/Mimosa_PIE_Recorder/WIDG_recorder_PlayerView_Mimosa.WIDG_recorder_PlayerView_Mimosa_C|GetPetDistanceIndicator _widgetplayer) "Hidden"))))

(event Custom|setVariables
  ;[0.6042 0.6042 0.6042] on start, establish variables then set self location based on settings
  ;;[1 1 1] set the player controller variable
  (bind _returnvalue (Actor|GetActorOfClass "/Game/Athena/Athena_PlayerController.Athena_PlayerController_C"))
  (Utilities|IsValid _returnvalue
    (:Is%20Valid
      (Variables|Default|SetplayerController _returnvalue))))

(event Custom|increaseCamDistance
  (bind _cameradistanceoverride (Variables|Properties|GetcameraDistanceOverride))
  (Variables|Properties|SetcameraDistanceOverride 15.0)
  (Utilities|FlowControl|Sequence
    (:then_0
      (CallFunction|SetCamDistance (Variables|Default|GetOffsetFront) _cameradistanceoverride))
    (:then_1
      (CallFunction|SetCamDistance (Variables|Default|GetOffset3/4R) _cameradistanceoverride))
    (:then_2
      (CallFunction|SetCamDistance (Variables|Default|GetOffset3/4L) _cameradistanceoverride))))

(event Custom|setCamera
  (bind _self self)
  (bind _facecamparent (Variables|Default|GetfaceCamParent))
  (bind _cameradistanceoverride (Variables|Properties|GetcameraDistanceOverride))
  ;[0.6042 0.6042 0.6042] on start, establish variables then set self location based on settings
  ;;[1 1 1] assign Actor variable, and ignore the rest of graph if we ONLY want gamecamera.
  (Utilities|IsValid (Pawn|GetControlledPawn (stale Custom|setVariables _returnvalue))
    (:Is%20Valid
      (bind _output_get (Variables|Default|Setactor (Pawn|GetControlledPawn (stale Custom|setVariables _returnvalue))))
      (Utilities|IsValid _output_get
        (:Is%20Valid
          (if (Variables|Properties|GetcharacterOnly)
            _
            (elif (Variables|Properties|GetsnapToPlayer)
              ;;[1 1 1] if we want the cameras to stick to the character in world space or not
              (Transformation|AttachActorToActor _self (Pawn|GetControlledPawn (stale Custom|setVariables _returnvalue)) "None" "SnapToTarget" "SnapToTarget" "KeepRelative" true)
              ;;[1 1 1] create and add the UI
              (bind _returnvalue (UserInterface|CreateWidget "/AnimInEdTools/Mimosa_PIE_Recorder/WIDG_recorder_PlayerView_Mimosa.WIDG_recorder_PlayerView_Mimosa_C" (Variables|Default|GetplayerController) (Variables|Properties|Getadditional3/4)))
              (UserInterface|Viewport|AddtoViewport _returnvalue 0)
              (bind _output_get_2 (Variables|Default|SetwidgetPlayer _returnvalue))
              ;;[1 1 1] adjust the size of the middle facecam in the widget.
              (if (Variables|properties|GetisBigFace)
                (Class|WIDGRecorderPlayerViewMimosa|BigFaceOn _output_get_2)
                (else
                  (Class|WIDGRecorderPlayerViewMimosa|BigFaceOff _output_get_2)))
              ;;[1 1 1] adjust the camera position if facecam on start
              (if (Variables|Properties|GetisFaceCam)
                (CallFunction|ToggleFacecam))
              ;;[1 1 1] adjust position and parent of the face camera
              (Transformation|AttachComponentToComponent _facecamparent (Actor|GetComponentbyClass (Variables|Default|Getactor) "/Script/Engine.SkeletalMeshComponent") "face_attach" "SnapToTarget" "SnapToTarget" "SnapToTarget" true)
              (Transformation|SetRelativeRotation _facecamparent "180.000000,90.000000,-90.000000" false false)
              (Transformation|SetRelativeLocation _facecamparent "5.000000,100.000000,0.000000" false false)
              (Utilities|FlowControl|Sequence
                (:then_0
                  (CallFunction|SetCamDistance (Variables|Default|GetOffsetFront) _cameradistanceoverride))
                (:then_1
                  (CallFunction|SetCamDistance (Variables|Default|GetOffset3/4R) _cameradistanceoverride))
                (:then_2
                  (CallFunction|SetCamDistance (Variables|Default|GetOffset3/4L) _cameradistanceoverride)))
              (else
                (Transformation|SetActorTransform _self (Transformation|GetActorTransform (Pawn|GetControlledPawn (stale Custom|setVariables _returnvalue))) false true)))))))))

(event Custom|decreaseCamDistance
  (bind _cameradistanceoverride (Variables|Properties|GetcameraDistanceOverride))
  (Variables|Properties|SetcameraDistanceOverride -15.0)
  (Utilities|FlowControl|Sequence
    (:then_0
      (CallFunction|SetCamDistance (Variables|Default|GetOffsetFront) _cameradistanceoverride))
    (:then_1
      (CallFunction|SetCamDistance (Variables|Default|GetOffset3/4R) _cameradistanceoverride))
    (:then_2
      (CallFunction|SetCamDistance (Variables|Default|GetOffset3/4L) _cameradistanceoverride))))

(event Custom|buttonPresses
  ;[1 1 1] overwrite current skin with selected from array
  (Variables|Default|GetplayerController
    (:then
      (Utilities|FlowControl|Sequence
        (:then_0
          (CallFunction|SetMood "One"))
        (:then_1
          (CallFunction|SetMood "NumPadOne"))
        (:then_2
          (CallFunction|SetMood "Two" 1))
        (:then_3
          (CallFunction|SetMood "NumPadTwo" 1))
        (:then_4
          (CallFunction|SetMood "Three" 2))
        (:then_5
          (CallFunction|SetMood "NumPadThree" 2))
        (:then_6
          (CallFunction|SetMood "Four" 3))
        (:then_7
          (CallFunction|SetMood "NumPadFour" 3))
        (:then_8
          (CallFunction|SetMood "Five" 4))
        (:then_9
          (CallFunction|SetMood "NumPadFive" 4))
        (:then_10
          (Utilities|ToggleAnimBP "Six"))
        (:then_11
          (CallFunction|RefreshCompanion "Zero"))
        (:then_12
          (CallFunction|RefreshCompanion "NumPadZero"))
        (:then_13
          ;[1 1 1] toggle facecam views
          (if (Game|Player|WasInputKeyJustPressed (Variables|Default|GetplayerController) "RightBracket")
            _)
          (Utilities|FlowControl|DoOnce false
            (:Completed
              (Class|WIDGRecorderPlayerViewMimosa|ToggleFaceCam (Variables|Default|GetwidgetPlayer)))))
        (:then_14
          ;[1 1 1] Comment
          (if (Game|Player|WasInputKeyJustPressed (Variables|Default|GetplayerController) "Multiply")
            _)
          (Utilities|FlowControl|DoOnce false))
        (:then_15
          (if (Game|Player|WasInputKeyJustPressed (Variables|Default|GetplayerController) "Add")
            _)
          (Utilities|FlowControl|DoOnce false
            (:Completed
              (CallFunction|IncreaseCamDistance))))
        (:then_16
          (if (Game|Player|WasInputKeyJustPressed (Variables|Default|GetplayerController) "Subtract")
            _)
          (Utilities|FlowControl|DoOnce false
            (:Completed
              (CallFunction|DecreaseCamDistance))))
        (:then_17
          (if (Game|Player|WasInputKeyJustPressed (Variables|Default|GetplayerController) "Backslash")
            _)
          (Utilities|FlowControl|DoOnce false
            (:Completed
              (Class|WIDGRecorderPlayerViewMimosa|Toggle4View (Variables|Default|GetwidgetPlayer)))))))
    (:else)))

(event OnComponentBeginOverlap(CompanionPetDistanceBubble)
  (params (OverlappedComponent :type Primitive%20Component%20Object%20Reference) (OtherActor :type Actor%20Object%20Reference) (OtherComp :type Primitive%20Component%20Object%20Reference) (OtherBodyIndex :type Integer) (bFromSweep :type Boolean) (SweepResult :type Hit%20Result%20Structure%20%28by%20ref%29))
  (Utilities|Casting|CastToPlayerController OtherActor)
  (Widget|SetVisibility (Variables|/AnimInEdTools/Mimosa_PIE_Recorder/WIDG_recorder_PlayerView_Mimosa.WIDG_recorder_PlayerView_Mimosa_C|GetPetDistanceIndicator (Variables|Default|GetwidgetPlayer)) "Visible"))

(fn refreshCompanion
  (params (ButtonPressed :type Key%20Structure))
  (if (Game|Player|WasInputKeyJustPressed (Variables|Default|GetplayerController) ButtonPressed)
    (if (Variables|Default|GetmidRefresh)
      (Development|ExecuteConsoleCommand "Fort.PlayerCompanion.bEnablePlayerCompanion 1")
      (Variables|Default|SetcompanionFound false)
      (Development|PrintString "Companion On")
      (Variables|Default|SetmidRefresh false)
      (else
        (Development|ExecuteConsoleCommand "Fort.PlayerCompanion.bEnablePlayerCompanion 0")
        (Development|PrintString "Companion Off")
        (Variables|Default|SetmidRefresh true)))))

(fn setCamDistance
  (params (camera :type Scene%20Component%20Object%20Reference) (offset :type Float%20%28double-precision%29))
  (Transformation|AddLocalOffset camera (Math|Vector|MakeVector offset) false false))

(fn setCompanion)

(fn setCompanionName
  (bind _companionname (Variables|Default|GetcompanionName))
  (Development|PrintString _companionname)
  (Widget|SetText(TextBox) (Variables|/AnimInEdTools/Mimosa_PIE_Recorder/WIDG_recorder_PlayerView_Mimosa.WIDG_recorder_PlayerView_Mimosa_C|Getskinname (Variables|Default|GetwidgetPlayer)) (Utilities|Text|ToText(String) _companionname)))

(fn setMood
  (params (buttonPressed :type Key%20Structure) (moodNumber :type Integer))
  (bind _moodcommands (Variables|Default|GetmoodCommands))
  (bind _output (Utilities|Array|Get(acopy) _moodcommands moodNumber))
  (if (Game|Player|WasInputKeyJustPressed (Variables|Default|GetplayerController) buttonPressed)
    (Development|ExecuteConsoleCommand _output)
    (Development|PrintString (Utilities|String|Append "Companion Mood Set: " (Utilities|String|RightChop _output 32)))))

(fn setSkin
  (params (skinNumber :type Integer) (buttonPressed :type Key%20Structure))
  (bind _skins (Variables|Properties|Getskins))
  (bind _output (Utilities|Array|Get(acopy) _skins skinNumber))
  (if (Game|Player|WasInputKeyJustPressed (Variables|Default|GetplayerController) buttonPressed)
    (Utilities|IsValid _output
      (:Is%20Valid
        (Items|ForceAthenaCosmetic _output)
        (Variables|Default|SetcurrentSkin _output)))))

(fn ConstructionScript
  (bind _facecamparent (Variables|Default|GetfaceCamParent))
  ;[1 1 1] on construction, remove the bottom right 3/4R camera if desired.
  (Variables|/Script/Engine.SceneCaptureComponent2D|SetTextureTarget (select (Variables|Properties|Getadditional3/4) "/AnimInEdTools/Mimosa_PIE_Recorder/RT_gameCamHalf.RT_gameCamHalf" "/AnimInEdTools/Mimosa_PIE_Recorder/RT_gameCamFull.RT_gameCamFull") (Variables|Default|GetgameCam))
  ;[1 1 1] set location of camera parent for preview
  (Transformation|SetRelativeLocation (Variables|Default|GetcameraParent) "0.000000,0.000000,60.000000" false false)
  ;[1 1 1] attach facecam to dummy for preview setup.
  (Transformation|AttachComponentToComponent _facecamparent (Variables|Default|GetcharBody) "head" "SnapToTarget" "SnapToTarget" "SnapToTarget" true)
  (Transformation|SetRelativeRotation _facecamparent "180.000000,90.000000,-90.000000" false false)
  (Transformation|SetRelativeLocation _facecamparent "4.000000,120.000000,0.000000" false false))
