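# UI-domain material for the Mimosa PIE Recorder tool, built around the RT_gameCamFull texture, which by its name is a full game-camera render target.
# Its graph holds a single linear-color Texture Sample of that texture, with no math, parameters or other nodes.
# The graph shows no link from the sample to a material output.
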Material /AnimInEdTools/Mimosa_PIE_Recorder/RT_gameCamFull_Mat
graph {
 properties {materialDomain=MD_UI, pixelDepthOffsetMode=PDOM_Legacy, usedShadingModels=Default Lit}
 nodes [TextureSample() {samplerType=SAMPLERTYPE_LinearColor, texture=&/RT_gameCamFull}]
}
