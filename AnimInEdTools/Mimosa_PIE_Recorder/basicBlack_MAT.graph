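# basicBlack_MAT is a User Interface domain material in the Mimosa PIE Recorder folder of AnimInEdTools, named as a plain black fill.
# Its graph holds a single Constant3Vector left at its default black value, with no connection shown to any material output.
# It has no textures, parameters or other logic.
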Material /AnimInEdTools/Mimosa_PIE_Recorder/basicBlack_MAT
graph {properties {materialDomain=MD_UI, pixelDepthOffsetMode=PDOM_Legacy, usedShadingModels=Default Lit}, nodes [Constant3Vector()]}
